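# Unlit translucent two-sided ribbon material for the Dark Bomber pickaxe's melee trail, combining glowing purple cube glyphs with dark, smoky color.
# Stepped panners on a cube height mask drive pulsing glyphs, added to a three-color blend of distorted noise, caustics and smoke plus an edge glow.
# Emissive gets particle color and time-of-day dimming at night.
# Opacity combines noise-broken sphere masks, particle Erode, near-camera fade, Fresnel and ribbon-end fades.
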Material /BRCosmetics/Weapons/FORT_Melee/Pickaxe_Dark_Bomber/FX/M_Melee_Trail_DarkBomber
graph {
 properties {
  bAllowNegativeEmissiveColor=true
  bAutomaticallySetUsageInEditor=false
  bUsedWithNiagaraRibbons=true
  blendMode=BLEND_Translucent
  pixelDepthOffsetMode=PDOM_Legacy
  shadingModel=MSM_Unlit
  translucencyLightingMode=TLM_VolumetricPerVertexDirectional
  twoSided=true
  usedShadingModels=Unlit
 }
 nodes [
  n1 = Multiply(A=Time(), B=ScalarParameter() {defaultValue=2.5, group=T H E C U B E, parameterName=Overall Emissive Pulse Speed})
  n2 = Multiply(
   A=ScalarParameter() {defaultValue=2, desc=1.75, group=T H E C U B E, parameterName=Glyph Independent Scale}
   B=TextureCoordinate() {coordinateIndex=1, vTiling=0.5}
  )
  n3 = ScalarParameter() {defaultValue=1, group=T H E C U B E, parameterName=Overall Texture Tiling}
  n4 = Ceil(None=Multiply(A=Time()) {constB=4})
  n5 = TextureSample(UVs=Panner(Coordinate=Multiply(A=@n3, B=@n2), Time=Divide(A=@n4)) {speedY=1}) {samplerType=SAMPLERTYPE_LinearColor, texture=/Game/Athena/Prototype/Blueprints/Cube/Materials/DFHeightMask}
  n6 = Multiply(
   A=@n3
   B=Multiply(
    A=ScalarParameter() {defaultValue=0.125, desc=0.125, group=T H E C U B E, parameterName=Glyph Panner Independent Scale}
    B=TextureCoordinate() {coordinateIndex=1, vTiling=0.25}
   )
  )
  n7 = Ceil(None=Multiply(A=Time()) {constB=8})
  n8 = TextureSample(UVs=Panner(Coordinate=@n6, Time=Divide(A=@n7) {constB=4}) {speedX=1}) {
   samplerSource=SSM_Wrap_WorldGroupSettings
   samplerType=SAMPLERTYPE_LinearColor
   texture=&DFHeightMask
  }
  n9 = Constant()
  n10 = LinearInterpolate(A=@n5.R, B=@n5.R, Alpha=If(A=@n8.B, B=Constant() {r=0.5}, "A > B"=Constant() {r=1}, "A == B"=@n9, "A < B"=@n9))
  n11 = Time()
  n12 = ScalarParameter() {defaultValue=0.3, group=T H E C U B E, parameterName=Glyph Pulse Speed}
  n13 = LinearInterpolate(
   A=ScalarParameter() {defaultValue=6, parameterName=Glyph Pulse Clamp A}
   B=ScalarParameter() {defaultValue=25, parameterName=Glyph Pulse Clamp B}
   Alpha=MaterialFunctionCall(Value=Multiply(A=@n11, B=@n12)) {materialFunction=&LinearSine}."Rounded Linear Sine "
  )
  n14 = ScalarParameter() {defaultValue=1, desc=0.0135, group=T H E C U B E, parameterName="glyph panner speed "}
  n15 = Time()
  n16 = Divide(A=Ceil(None=Multiply(A=@n15) {constB=4}))
  n17 = TextureSample!(
   UVs=Panner(Coordinate=@n6, Time=Multiply(A=@n14, B=@n16)) {bFractionalPart=true, speedY=-1}
   Tex
   Bias=ScalarParameter() {defaultValue=2, parameterName=mip bias Glyph HOrizontal Panner}
   Apply View MipBias
  ) {
   mipValueMode=TMVM_MipBias
   samplerSource=SSM_Wrap_WorldGroupSettings
   samplerType=SAMPLERTYPE_LinearColor
   texture=&DFHeightMask
  }
  n18 = Multiply(A=@n10, B=Saturate(None=LinearInterpolate(Alpha=@n17.B) {constA=1, constB=0}))
  n19 = Constant() {r=1.25}
  n20 = LinearInterpolate(B=Multiply(A=Power(Base=@n10, Exp=@n13)) {constB=5}, Alpha=Saturate(None=Multiply(A=@n18, B=@n19))) {constB=0}
  n21 = Multiply(
   A=LinearInterpolate(
    A=ScalarParameter() {defaultValue=0.4, group=T H E C U B E, parameterName=Overall Emissive Min}
    B=ScalarParameter() {defaultValue=0.6, group=T H E C U B E, parameterName=Overall Emissive Max}
    Alpha=Sine(None=@n1)
   )
   B=Multiply(
    A=VectorParameter() {group=T H E C U B E, parameterName=Glyph Glow Color, defaultValue=<rgba 0.4689 0.042 0.8958 1>}.RGB
    B=Saturate(None=@n20)
   )
  )
  n22 = OneMinus(None=ComponentMask(None=TextureCoordinate()) {g=true})
  n23 = VectorParameter() {parameterName=Color_A, defaultValue=<rgba -0.05 -0.05 -0.05 1>}
  n24 = VectorParameter() {parameterName=Color_B, defaultValue=<rgba 0.0025 0.0026 0.026 1>}
  n25 = VectorParameter() {parameterName=Color_C, defaultValue=<rgba 0.062 0.0145 0.15 1>}
  n26 = OneMinus(None=ComponentMask(None=TextureCoordinate()) {g=true})
  n27 = TextureSample!(UVs=Panner(Coordinate=TextureCoordinate() {coordinateIndex=1, uTiling=2}) {speedX=-0.4}, Tex, Apply View MipBias) {texture=&T_NoisePack_01}
  n28 = ShadingPathSwitch(
   Default=Divide(A=ComponentMask(None=@n27.RGB) {g=true, r=true}, B=ScalarParameter() {defaultValue=8, parameterName=UVdistortAmount}) {constB=8}
   Mobile=Constant()
  ) {desc=Migrated from FeatureLevelSwitch}
  n29 = ScalarParameter() {parameterName=PanSpeed}
  n30 = Add(A=@n29) {constB=0.3}
  n31 = Constant()
  n32 = Time()
  n33 = StaticSwitchParameter(True=Add(A=DynamicParameter() {defaultValue=<rgba 1 1 1 0>, paramNames [TimeOffset,Random,Desa,Erode]}.TimeOffset, B=@n32), False=@n32) {defaultValue=true, parameterName=UseTimeOffset}
  n34 = Multiply(A=TextureCoordinate() {coordinateIndex=1, vTiling=0.7}, B=ScalarParameter() {defaultValue=1, parameterName=NoiseTile})
  n35 = Add(A=@n28, B=Panner(Coordinate=@n34, Time=@n33, Speed=AppendVector(A=@n30, B=@n31)) {speedX=0.3})
  n36 = Add(A=@n29) {constB=0.2}
  n37 = Add(A=@n28, B=Panner(Coordinate=@n34, Time=@n33, Speed=AppendVector(A=@n36, B=@n31)) {speedX=0.2})
  n38 = MaterialFunctionCall(Float2=@n34) {materialFunction=&BreakOutFloat2Components}
  n39 = Add(A=@n38.G) {constB=0.5}
  n40 = Add(A=@n28, B=Panner(Coordinate=AppendVector(A=@n38.R, B=@n39), Time=@n33, Speed=AppendVector(A=@n29, B=@n31)))
  n41 = Add(
   A=LinearInterpolate(Alpha=TextureSample(UVs=@n37) {texture=&T_ShieldCaustics}.R) {constA=0.2, constB=3}
   B=LinearInterpolate(Alpha=TextureSample(UVs=@n40) {texture=&T_Inky_Smoke_Tile}.R) {constB=3}
  )
  n42 = LinearInterpolate(
   B=Desaturation(
    None=TextureSample(UVs=@n35) {texture=&T_NoisePack_01}.RGB
    Fraction=DynamicParameter() {defaultValue=<rgba 1 1 1 0>, paramNames [TimeOffset,Random,Desa,Erode]}.Desa
   )
   Alpha=@n41
  )
  n43 = Saturate(None=Add(A=Multiply(A=@n26, B=@n42)) {constB=0.0625})
  n44 = TextureCoordinate()
  n45 = Power(Base=OneMinus(None=ComponentMask(None=@n44) {g=true})) {constExponent=25}
  n46 = Add(
   A=Add(
    A=Multiply(A=@n21, B=@n22)
    B=MaterialFunctionCall(A=@n23.RGB, B=@n24.RGB, C=@n25.RGB, Alpha=@n43) {materialFunction=&3ColorBlend}.Result
   )
   B=Multiply(A=VectorParameter() {group=T H E C U B E, parameterName=Glyph Glow Color, defaultValue=<rgba 0.4689 0.042 0.8958 1>}.RGB, B=@n45)
  )
  n47 = ParticleColor()
  n48 = Multiply(
   A=MaterialFunctionCall(LDR Input=Clamp(None=@n46) {minDefault=-1}, HDR Tint=@n46) {materialFunction=&DeriveHDRfromLDR}.RemappedRange
   B=@n47.RGB
  )
  MaterialFunctionCall(Emissive=MaterialFunctionCall(Color_Morning=@n48, Color_Day=@n48, Color_Evening=@n48, Color_Night=Multiply(A=@n48) {constB=0.25}) {materialFunction=&Color_Time_of_Day}.Result) {materialFunction=&MF_FX_EmissiveAdjustments}
  n49 = ConstantBiasScale(None=@n41) {bias=-0.5, scale=0.2}
  n50 = OneMinus(None=ComponentMask(None=TextureCoordinate()) {r=true})
  n51 = Constant() {r=0.5}
  n52 = AppendVector(
   A=@n51
   B=MaterialFunctionCall(Float2=TextureCoordinate()) {materialFunction=&BreakOutFloat2Components}.G
  )
  n53 = ScalarParameter() {defaultValue=0.5, parameterName=Mask Radius}
  n54 = ScalarParameter() {defaultValue=0.9, parameterName=Mask Hardness}
  n55 = OneMinus(None=SphereMask(A=TextureCoordinate() {uTiling=0.5}, B=Constant2Vector() {g=0.82, r=0.85}, Radius=Constant() {r=0.8}, Hardness=Constant()))
  n56 = Multiply(
   A=Saturate(None=Add(A=@n49, B=@n50))
   B=Multiply(A=SphereMask(A=@n52, B=@n51, Radius=@n53, Hardness=@n54), B=Power(Base=@n55) {constExponent=5})
  )
  n57 = Subtract(A=Saturate(None=Add(A=@n45, B=@n56)), B=DynamicParameter() {defaultValue=<rgba 1 1 1 0>, paramNames [TimeOffset,Random,Desa,Erode]}.Erode)
  n58 = Multiply(A=@n47.A, B=Clamp(None=@n57, Max=@n57))
  n59 = MaterialFunctionCall(Distance=ScalarParameter() {defaultValue=64, parameterName=NearFadeDistance}) {materialFunction=&Near_Camera_Fade}
  n60 = ScalarParameter() {defaultValue=1.5, parameterName=Fresnel Brightness}
  n61 = ScalarParameter() {defaultValue=3, parameterName=Fresnel Exponent}
  n62 = ComponentMask(None=TextureCoordinate()) {r=true}
  n63 = OneMinus(None=@n62)
  Multiply(
   A=Multiply(
    A=Multiply(A=@n58, B=@n59.Result)
    B=MaterialFunctionCall(Brightness=@n60, Exponent=@n61) {materialFunction=&MF_Fresnel_FX}.Fresnel Double Sided
   )
   B=Multiply(A=Multiply(A=@n62, B=@n63)) {constB=2}
  )
 ]
}
