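# Physics asset for the Dark Bomber pickaxe skeletal mesh, a two-handed melee cosmetic weapon in BRCosmetics.
# It defines a single body on the Melee_TwoHanded_Handle bone, shaped as one offset, rotated capsule that approximates the weapon.
# It has no constraints, so the pickaxe simulates and collides as one rigid piece with no articulated parts.
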PhysicsAsset /BRCosmetics/Weapons/FORT_Melee/Pickaxe_Dark_Bomber/Meshes/SK_Pickaxe_Dark_Bomber_PhysicsAsset
bodies {Melee_TwoHanded_Handle [{shapeType=Capsule, length=72.7229, radius=22.6193, center=<0.8107 1.8806 34.553>, rotation=<rot 0.1344 -90 -0.0185>}]}
constraints []
